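; Fortnite Artemis big clock building actor whose hands can follow the game's time of day, with audio.
; Construction uses the update flag to swap hand materials and set audio activation.
; When enabled, cosmetic-capable instances run a looping timer that writes normalized time of day to the hour hand's custom primitive data and plays the audio.
; Death effects fade the audio out.
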(event EventBeginPlay
  (Networking|CanExecuteCosmeticEvents
    (:True
      (if (Variables|Clock|GetBShouldUpdateTime)
        (Utilities|Time|SetTimerbyFunctionName 0 "UpdateTime" 1.0 true false 1.0 0.5)
        (CallFunction|UpdateTime)))))

(event EventOnDeathPlayEffects
  (params (Damage :type Float%20%28single-precision%29) (DamageTags :type Gameplay%20Tag%20Container%20Structure%20%28by%20ref%29) (Momentum :type Vector :default "0, 0, 0") (HitInfo :type Hit%20Result%20Structure%20%28by%20ref%29) (InstigatedBy :type Fort%20Pawn%20Object%20Reference) (DamageCauser :type Actor%20Object%20Reference) (EffectContext :type Gameplay%20Effect%20Context%20Handle%20Structure))
  (bind _audio (Variables|Default|GetAudio))
  ;[1 1 1] On Death Fade Out Audio
  (Utilities|IsValid _audio
    (:Is%20Valid
      (if (Utilities|IsValid _audio :!class /Script/Engine.KismetSystemLibrary)
        (Audio|Components|Audio|FadeOut (Variables|Default|GetAudio) 0.2 0.0 "Linear" :!class /Script/Engine.AudioComponent)))))

(fn setTickingEnabled
  (params (enabled :type Boolean))
  (Variables|Clock|SetBShouldUpdateTime enabled)
  (Utilities|FlowControl|Sequence
    (:then_0
      (if (Variables|Clock|GetBShouldUpdateTime)
        (Rendering|Material|SetMaterial (Variables|BigClock_01|GetMinuteHandMesh) 0 "/Game/Environments/Artemis/Props/BigClock/Materials/MI_Artemis_ClockHand_TOD_Min.MI_Artemis_ClockHand_TOD_Min")
        (Rendering|Material|SetMaterial (Variables|BigClock_01|GetHourHandMesh) 0 "/Game/Environments/Artemis/Props/BigClock/Materials/MI_Artemis_ClockHand_TOD_Hour.MI_Artemis_ClockHand_TOD_Hour")
        (Transformation|SetRelativeRotation (Variables|BigClock_01|GetHourHandMesh) "0, 0, 0" false false)
        (else
          (Rendering|Material|SetMaterial (Variables|BigClock_01|GetHourHandMesh) 0 "/Game/Environments/Props/Clocks/Materials/MI_BigClock_01.MI_BigClock_01"))))
    (:then_1
      (Components|Activation|SetActive (Variables|Default|GetAudio) (Variables|Clock|GetBShouldUpdateTime) false))))

(fn UpdateTime
  (Rendering|Material|SetCustomPrimitiveDataFloat (Variables|BigClock_01|GetHourHandMesh) 0 (/ (TimeOfDay|GetTimeOfDay :!class /Script/FortniteGame.FortKismetLibrary) 24.0))
  (Audio|Components|Audio|Play (Variables|Default|GetAudio) 0.0))

(fn ConstructionScript
  (CallFunction|SetTickingEnabled (Variables|Clock|GetBShouldUpdateTime)))
